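; Blueprint function library of lookup helpers for environmental sprites, covering sprite type, color and inventory item.
; Item type resolves by matching the item's gameplay tags against the keys of a tag-to-type map held as a local variable, falling back to the first enum entry.
; Inventory search returns the first item tagged Item.Sprite.
; Color comes from a local type-to-color map, with magenta returned for missing types.
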(fn FindItemSpriteType
  (params (Item :type Fort%20Item%20Definition%20Object%20Reference))
  (Utilities|IsValid Item
    (:Is%20Valid
      (bind _keys (Utilities|Map|Keys (CallFunction|GetTagstoTypeMap (Variables|Get__WorldContext))))
      (for _array_element _keys
        (if (GameplayTags|HasMatchingGameplayTag Item _array_element)
          (return (CallFunction|GetSpriteTypefromTag _array_element (Variables|Get__WorldContext)))))
      (return "NewEnumerator0"))))

(fn FindSpriteItemInInventory
  (params (InventoryOwner :type Fort%20Inventory%20Owner%20Interface%20Interface))
  (Utilities|IsValid (Utilities|ToObject(Interface) InventoryOwner)
    (:Is%20Valid
      (bind _items (InventoryManagement|GetItemsfromInventoryOwner InventoryOwner))
      (bind _array_element (Utilities|Array|ForEachLoopwithBreak _items)
        (:LoopBody
          (if (GameplayTags|HasMatchingGameplayTag _array_element "(TagName=\"Item.Sprite\")")
            (return _array_element)))
        (:Completed
          _)))
    (:Is%20Not%20Valid
      _))
  (return 0))

(fn GetSpriteColor
  (params (SpriteType :type EEnvironmentalSpriteType%20Enum :default "ESpriteType_MAX"))
  (bind _colordictionary (LocalVariables|GetColorDictionary))
  (bind (_value _returnvalue) (Utilities|Map|Find _colordictionary SpriteType))
  (if _returnvalue
    (return _value)
    (else
      (return "(R=1.000000,G=0.000000,B=1.000000,A=1.000000)"))))

(fn GetSpriteTypeFromTag
  (params (Key :type Gameplay%20Tag%20Structure%20%28by%20ref%29))
  (bind _worldcontext (Variables|Get__WorldContext))
  (bind _tagstotype (CallFunction|GetTagstoTypeMap _worldcontext))
  (bind (_value _returnvalue) (Utilities|Map|Find _tagstotype Key))
  (if _returnvalue
    (return _value)
    (else
      (return "NewEnumerator0"))))

(fn GetTagsToTypeMap
  (return (LocalVariables|GetTagsToTypeMap)))
