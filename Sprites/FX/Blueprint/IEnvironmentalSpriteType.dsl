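; IEnvironmentalSpriteType is, by its name and path, a Blueprint interface for environmental sprite effects that exposes a sprite's type.
; Its single function, GetSpriteType, returns an enumerator value, shown here only as the first entry, NewEnumerator0.
; The file contains no other logic and does not show the enum's display names or which objects implement or call the function.
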(fn GetSpriteType
  (return "NewEnumerator0"))
