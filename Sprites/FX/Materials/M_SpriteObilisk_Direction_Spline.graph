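# Translucent, unlit, two-sided material for Niagara mesh particles, apparently a directional spline streak for the sprite obelisk effect.
# SplineThicken sets ribbon width from a dynamic parameter, with a Bend vector adding an arched offset along vertex normals.
# Panning smoke energy, two sparkle layers and a distorted SphereMask sweep, positioned by another dynamic parameter, form named masks.
# Particle color tints the glow, while edge and end fades plus time-of-day brightness values support opacity shaping.
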Material /Sprites/FX/Materials/M_SpriteObilisk_Direction_Spline
graph {
 properties {
  bAutomaticallySetUsageInEditor=false
  bUsedWithNiagaraMeshParticles=true
  blendMode=BLEND_Translucent
  shadingModel=MSM_Unlit
  twoSided=true
  usedShadingModels=Unlit
 }
 nodes [
  n1 = VertexNormalWS()
  n2 = VectorParameter() {parameterName=Bend}
  n3 = ComponentMask(None=TextureCoordinate()) {g=true}
  n4 = OneMinus(None=@n3)
  n5 = DynamicParameter() {parameterIndex=2, paramNames [Sweep, Spline Thickness, Param3, Param4]}
  n6 = MaterialFunctionCall(
   WidthBase=ScalarParameter() {defaultValue=24, parameterName=Spline Base}
   WidthTip=Multiply(A=ScalarParameter() {defaultValue=1, parameterName=Spline Width}, B=@n5.Spline Thickness)
  ) {materialFunction=&SplineThicken}
  Add(A=Multiply(A=Multiply(A=@n1, B=@n2.RGB), B=Multiply(A=@n3, B=@n4)), B=@n6.WorldPositionOffset)
  Add(B=ComponentMask(None=TextureCoordinate()) {g=true})
  Add()
  n7 = ComponentMask(None=TextureCoordinate()) {g=true}
  n8 = OneMinus(None=@n7)
  LinearInterpolate(Alpha=Power(Base=Multiply(A=@n7, B=@n8), Exp=ScalarParameter() {defaultValue=0.5, parameterName=Param_2}))
  MaterialFunctionCall() {materialFunction=&Swizzle}
  n9 = Multiply(A=OneMinus(None=NamedRerouteUsage()), B=ScalarParameter() {defaultValue=1, parameterName=Param_1})
  n10 = ScalarParameter() {defaultValue=1}
  n11 = ParticleColor()
  n12 = Power(Base=ComponentMask(None=TextureCoordinate()) {g=true}, Exp=ScalarParameter() {defaultValue=4, parameterName=Param_7})
  n13 = Add(A=@n11.RGB, B=Multiply(A=Multiply(A=@n11.RGB, B=@n12), B=ScalarParameter() {defaultValue=10, parameterName=Param_6}))
  n14 = NamedRerouteUsage()
  MaterialFunctionCall(Emissive=Add(A=Power(Base=@n9, Exp=@n10), B=Multiply(A=@n13, B=@n14))) {materialFunction=&MF_FX_EmissiveAdjustments}
  Multiply()
  n15 = Add(A=TextureCoordinate(), B=Multiply(A=NamedRerouteUsage(), B=ScalarParameter() {defaultValue=0.005, parameterName=Energy Distortion Amnt}))
  Multiply(B=Power(Base=ComponentMask(None=@n15) {g=true}, Exp=ScalarParameter() {defaultValue=1, parameterName=Param_4}))
  Multiply(B=@n5.Spline Thickness)
  n16 = VertexNormalWS()
  n17 = NamedRerouteUsage()
  Multiply(
   A=Multiply(
    A=NamedRerouteUsage()
    B=MaterialFunctionCall(BaseNormal=@n16, AdditionalNormal=@n17) {materialFunction=&BlendAngleCorrectedNormals}.Result
   )
   B=ScalarParameter() {parameterName=wpo}
  )
  n18 = TextureCoordinate()
  n19 = OneMinus(None=Power(Base=ComponentMask(None=@n18) {g=true}, Exp=ScalarParameter() {defaultValue=4, parameterName=Spline End Fade}))
  n20 = ComponentMask(None=TextureCoordinate()) {r=true}
  n21 = Multiply(A=@n14, B=Multiply(A=@n20, B=OneMinus(None=@n20)))
  n22 = Multiply(A=Multiply(A=Multiply(A=@n19, B=@n21), B=ParticleColor().A), B=NamedRerouteUsage())
  n23 = NamedRerouteUsage()
  n24 = TextureCoordinate()
  Multiply(
   A=Multiply(A=Multiply(A=@n22, B=@n23), B=ScalarParameter() {defaultValue=20, parameterName=Param_3})
   B=Power(Base=ComponentMask(None=@n24) {g=true}, Exp=ScalarParameter() {defaultValue=1, parameterName=Spline Fade})
  )
  n25 = Add(A=TextureCoordinate(), B=ParticleRandom())
  n26 = ComponentMask(None=VectorParameter() {parameterName=Secondary Energy Tiling, defaultValue=<rgba 1 1.25 0 0>}.RGB) {g=true, r=true}
  n27 = Add(A=Time(), B=ComponentMask(None=TextureCoordinate() {coordinateIndex=2}) {g=true})
  n28 = ScalarParameter() {defaultValue=0.5, parameterName=Secondary Energy Pan Speed}
  n29 = MaterialFunctionCall(XY=Panner(Coordinate=Multiply(A=@n25, B=@n26), Time=Multiply(A=@n27, B=@n28)) {speedY=1}) {materialFunction=&Swizzle}
  NamedRerouteDeclaration(None=Multiply(
    A=TextureSample(UVs=@n29.YX) {samplerType=SAMPLERTYPE_Grayscale, texture=&T_Smoke_Trail_Soft}.G
    B=ScalarParameter() {defaultValue=1, parameterName=Param_13}
   )) {name=Energy, nodeColor=<rgba 1 0.7266 0 1>}
  n30 = TextureCoordinate()
  n31 = ComponentMask(None=VectorParameter() {parameterName=Sparkle Smaller Tiling, defaultValue=<rgba 5 10 0 0>}.RGB) {g=true, r=true}
  n32 = Add(A=Time(), B=ComponentMask(None=TextureCoordinate() {coordinateIndex=2}) {g=true})
  n33 = ScalarParameter() {defaultValue=1, parameterName=Sparkle Smaller Pan Speed}
  n34 = TextureSample(UVs=Panner(Coordinate=Multiply(A=@n30, B=@n31), Time=Multiply(A=@n32, B=@n33)) {speedY=1}) {texture=/Sprites/FX/Materials/Textures/T_EnvSprite_Glisten}
  n35 = ScalarParameter() {defaultValue=5, parameterName=Param_12}
  n36 = TextureCoordinate()
  n37 = ComponentMask(None=VectorParameter() {parameterName=Sparkle Tiling, defaultValue=<rgba 2.5 5 0 0>}.RGB) {g=true, r=true}
  n38 = ScalarParameter() {defaultValue=1, parameterName=Sparkle Pan Speed}
  n39 = TextureSample(UVs=Panner(Coordinate=Multiply(A=@n36, B=@n37), Time=Multiply(A=@n32, B=@n38)) {speedY=1}) {texture=/Sprites/FX/Materials/Textures/T_EnvSprite_FinerSparkles}
  n40 = ScalarParameter() {defaultValue=5, parameterName=Param_11}
  NamedRerouteDeclaration(None=Add(A=Multiply(A=@n34.G, B=@n35), B=Multiply(A=@n39.G, B=@n40))) {name=Sparkles, nodeColor=<rgba 0.6875 0 1 1>}
  n41 = Multiply(A=NamedRerouteUsage(), B=ScalarParameter() {defaultValue=0.5, parameterName=Sweep Edge Distortion})
  n42 = ComponentMask(None=TextureCoordinate()) {g=true}
  NamedRerouteDeclaration(None=SphereMask(
    A=Add(A=@n41, B=@n42)
    B=DynamicParameter() {parameterIndex=2, paramNames [Sweep, Spline Thickness, Param3, Param4]}.Sweep
    Radius=ScalarParameter() {defaultValue=0.25, parameterName=Param_10}
   ) {attenuationRadius=0.5, hardnessPercent=0}) {name=Sweep, nodeColor=<rgba 0 1 0.7891 1>}
  NamedRerouteDeclaration(None=@n6.Normal) {name=New Spline Normal, nodeColor=<rgba 0 0.9062 1 1>}
  NamedRerouteDeclaration(None=MaterialFunctionCall(
    Value_Morning=ScalarParameter() {defaultValue=0.2, parameterName=Opacity Morning}
    Value_Day=ScalarParameter() {defaultValue=0.3, parameterName=Opacity Day}
    Value_Evening=ScalarParameter() {defaultValue=0.25, parameterName=Opacity Evening}
    Value_Night=ScalarParameter() {defaultValue=0.25, parameterName=Opacity Night}
   ) {materialFunction=&Scalar_Time_of_Day}.Result) {name=TOD Brightness, nodeColor=<rgba 0 0.9766 1 1>}
  OneMinus()
  OneMinus()
  ScalarParameter() {parameterName=Param_9}
  Time()
  Time()
  Time()
 ]
}
