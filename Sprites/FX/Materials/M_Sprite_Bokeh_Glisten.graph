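# Additive unlit Niagara mesh-particle material that renders glistening bokeh sprites, billboarded through the Sprite function.
# Vertex color randomizes each sprite's looping time phase, scale, camera-depth offset and hue shift of the particle color.
# The loop drives sprite scale, with a sphere mask shaping opacity alongside the bokeh texture.
# Player-position sphere masks confine visibility to a distance band, and opacity varies by time of day.
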Material /Sprites/FX/Materials/M_Sprite_Bokeh_Glisten
graph {
 properties {
  bAutomaticallySetUsageInEditor=false
  bUsedWithNiagaraMeshParticles=true
  blendMode=BLEND_Additive
  shadingModel=MSM_Unlit
  twoSided=true
  usedShadingModels=Unlit
 }
 nodes [
  n1 = ScalarParameter() {defaultValue=10, parameterName=Random Scale Min}
  n2 = ScalarParameter() {defaultValue=20, parameterName=Random Scale}
  n3 = VertexColor()
  n4 = Time()
  n5 = ScalarParameter() {defaultValue=0.25, parameterName=timescale}
  n6 = Fmod(A=Add(A=@n3.A, B=Multiply(A=@n4, B=@n5)), B=ScalarParameter() {defaultValue=1, parameterName=Param_1})
  n7 = MaterialFunctionCall!(
   Center Location=WorldPosition().XYZ
   0-1 UVs
   XY scale=Multiply(A=LinearInterpolate(A=@n1, B=@n2, Alpha=@n3.G), B=@n6)
   OverridePivot
   "Normals (Optional)"
  ) {materialFunction=&Sprite}
  n8 = ScalarParameter() {defaultValue=-150, parameterName=Param_8}
  n9 = ScalarParameter() {defaultValue=150, parameterName=Param_7}
  n10 = VertexColor()
  Add(
   A=@n7.World position offset
   B=MaterialFunctionCall(Offset Amount=LinearInterpolate(A=@n8, B=@n9, Alpha=@n10.G), World Position=WorldPosition().XYZ) {materialFunction=/Engine/Functions/Engine_MaterialFunctions02/WorldPositionOffset/CameraOffset}.Camera Offset
  )
  n11 = DynamicParameter() {parameterIndex=1, paramNames [X,Y,Z,Param4]}
  n12 = MaterialFunctionCall(X=@n11.X, Y=@n11.X, Z=@n11.X) {materialFunction=&MakeFloat3}
  Add(
   A=LinearInterpolate(B=Transform(None=@n12.Result) {transformSourceType=TRANSFORMSOURCE_World, transformType=TRANSFORM_Instance}, Alpha=VertexColor().G)
   B=@n7.World position offset
  )
  n13 = ScalarParameter() {defaultValue=0.1, parameterName=Param_2}
  n14 = ScalarParameter() {parameterName=Param_3}
  n15 = Round(None=VertexColor().R)
  n16 = ParticleColor()
  MaterialFunctionCall(Emissive=MaterialFunctionCall(Hue Shift Percentage=LinearInterpolate(A=@n13, B=@n14, Alpha=@n15), Texture=@n16.RGB) {materialFunction=&HueShift}.Result) {materialFunction=&MF_FX_EmissiveAdjustments}
  Multiply()
  n17 = CollectionParameter() {collection=&FortniteMaterialParameters, parameterName=PlayerPosition}
  n18 = SphereMask(
   A=ComponentMask(None=@n17) {b=true, g=true, r=true}
   B=ParticlePositionWS()
   Radius=ScalarParameter() {defaultValue=750, parameterName=Param_6}
  ) {hardnessPercent=35}
  n19 = CollectionParameter() {collection=&FortniteMaterialParameters, parameterName=PlayerPosition}
  n20 = OneMinus(None=SphereMask(
    A=ComponentMask(None=@n19) {b=true, g=true, r=true}
    B=ParticlePositionWS()
    Radius=ScalarParameter() {defaultValue=500, parameterName=BokehHide}
   ) {hardnessPercent=50})
  n21 = Multiply(
   A=LinearInterpolate(B=ScalarParameter() {defaultValue=0.35, parameterName=Bokeh Fade}, Alpha=Multiply(A=@n18, B=@n20)) {constAlpha=1}
   B=Multiply(
    A=TextureSample() {samplerType=SAMPLERTYPE_LinearColor, texture=/Game/TimeOfDay/TODM/BR/s11/ExposureAssets/FN_BokehS11}.G
    B=ScalarParameter() {defaultValue=1}
   )
  )
  n22 = SphereMask(
   A=NamedRerouteUsage()
   B=ScalarParameter() {parameterName=StartPoint}
   Radius=ScalarParameter() {defaultValue=1, parameterName=Param_4}
   Hardness=ScalarParameter() {parameterName=Param_5}
  )
  Multiply(
   A=Multiply(A=@n16.G, B=Multiply(A=@n21, B=@n22))
   B=MaterialFunctionCall(
    Value_Morning=ScalarParameter() {defaultValue=2, parameterName=Opacity Morning}
    Value_Day=ScalarParameter() {defaultValue=1.5, parameterName=Opacity Day}
    Value_Evening=ScalarParameter() {defaultValue=1.5, parameterName=Opacity Evening}
    Value_Night=ScalarParameter() {defaultValue=1.5, parameterName=Opacity Night}
   ) {materialFunction=&Scalar_Time_of_Day}.Result
  )
  NamedRerouteDeclaration(None=@n6) {name=Lifetime, nodeColor=<rgba 0 1 0.6484 1>}
  SphereMask(
   A=TextureCoordinate()
   B=Constant() {r=0.5}
   Radius=ScalarParameter() {defaultValue=0.5, parameterName=Radius}
   Hardness=ScalarParameter() {parameterName=Softness}
  )
 ]
}
