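# Translucent unlit Niagara mesh-particle material for an Earth sprite wisp trail, colored by particle color through an emissive adjustment function.
# A scrolling smoke texture, offset per particle with optional UV swizzle, is shaped by a center falloff, a powered highlight and optional edge fade.
# An optional lead mask, panned and scaled by dynamic parameters, multiplies that opacity before depth fade.
# SplineThicken sets trail thickness from base and tip widths.
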Material /Sprites/FX/Materials/M_Sprites_Earth_WispTrail
graph {
 properties {
  bAutomaticallySetUsageInEditor=false
  bUsedWithNiagaraMeshParticles=true
  blendMode=BLEND_Translucent
  shadingModel=MSM_Unlit
  usedShadingModels=Unlit
 }
 nodes [
  n1 = Constant() {r=1}
  n2 = DynamicParameter() {paramNames [LeadTexturePan, LeadTextureYSize, Param3, Param4]}
  n3 = Multiply(A=TextureCoordinate(), B=AppendVector(A=Constant() {r=1}, B=@n2.LeadTextureYSize))
  n4 = AppendVector(A=Constant(), B=LinearInterpolate(Alpha=@n2.LeadTexturePan) {constA=-1})
  n5 = Multiply(
   A=TextureSample(UVs=Add(A=@n3, B=@n4)) {samplerType=SAMPLERTYPE_LinearGrayscale, texture=/Sprites/FX/Textures/T_Sprites_Earth_TrailMask}.R
   B=ScalarParameter() {defaultValue=1, parameterName=LeadTextureContrast}
  )
  n6 = ScalarParameter() {defaultValue=1, parameterName=UseLeadTexture}
  n7 = ParticleColor()
  n8 = TextureCoordinate()
  n9 = Add(A=@n8, B=Multiply(A=Constant2Vector() {g=1}, B=ParticleRandom()))
  n10 = LinearInterpolate(
   A=@n9
   B=MaterialFunctionCall(XYZ=@n9) {materialFunction=&Swizzle}.YX
   Alpha=ScalarParameter() {defaultValue=1, parameterName=SwizzleUV}
  )
  n11 = VectorParameter() {parameterName=ScrollSpeed, defaultValue=<rgba -1 0 3 1>}
  n12 = AppendVector(A=@n11.B, B=@n11.B)
  n13 = TextureSampleParameter2D(UVs=Panner(Coordinate=Multiply(A=@n10, B=@n12), Speed=ComponentMask(None=@n11.RGB) {g=true, r=true}) {speedX=-1}) {parameterName=MainTexture, samplerType=SAMPLERTYPE_Grayscale, texture=&T_Smoke_Trail_Soft}
  n14 = Abs(None=Subtract(A=ComponentMask(None=@n8) {g=true}) {constB=0.5})
  n15 = Power(Base=OneMinus(None=Multiply(A=@n14) {constB=2})) {constExponent=1}
  n16 = Power(Base=@n13.R) {constExponent=6}
  n17 = Multiply(A=@n7.A, B=Add(A=Multiply(A=@n13.R, B=@n15), B=Multiply(A=@n16) {constB=3}))
  n18 = ComponentMask(None=TextureCoordinate()) {g=true}
  n19 = Multiply(A=@n17, B=Power(Base=OneMinus(None=@n18)) {constExponent=1.4})
  n20 = ScalarParameter() {parameterName=TurnOnEdgeMask}
  DepthFade(Opacity=Multiply(A=LinearInterpolate(A=@n1, B=@n5, Alpha=@n6), B=LinearInterpolate(A=@n17, B=@n19, Alpha=@n20))) {fadeDistanceDefault=15}
  MaterialFunctionCall(WidthBase=ScalarParameter() {defaultValue=25, parameterName=WidthBase}, WidthTip=ScalarParameter() {defaultValue=40, parameterName=WidthTip}) {materialFunction=&SplineThicken}
  MaterialFunctionCall(Emissive=@n7.RGB) {materialFunction=&MF_FX_EmissiveAdjustments}
 ]
}
