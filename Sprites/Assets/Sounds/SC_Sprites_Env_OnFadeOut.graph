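# Sound cue for environmental sprites fading out, playing a one-shot disappear chime with spatial attenuation.
# A quality-level node routes the first two levels to a random pick from ten chimes and the last two to a four-chime subset.
# A root modulator varies the chosen sound, with a lowered volume floor for wide loudness variation.
# A concurrency override caps simultaneous instances at two, stopping the oldest.
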SoundCue /Sprites/Assets/Sounds/SC_Sprites_Env_OnFadeOut
root=Modulator_0
nodes [
 {id=Modulator_0, children [QualityLevel_0], properties {volumeMin=0.3}}
 {id=QualityLevel_0, children [Random_1,Random_1,Random_2,Random_2]}
 {
  id=Random_1
  children [
   WavePlayer_4
   WavePlayer_5
   WavePlayer_6
   WavePlayer_7
   WavePlayer_8
   WavePlayer_9
   WavePlayer_10
   WavePlayer_11
   WavePlayer_12
   WavePlayer_13
  ]
  properties {bShouldExcludeFromBranchCulling=true, weights [1,1,1,1,1,1,1,1,1,1]}
 }
 {id=Random_2, children [WavePlayer_17, WavePlayer_16, WavePlayer_15, WavePlayer_14], properties {weights [1,1,1,1]}}
 {id=WavePlayer_10, properties {soundWaveAssetPtr=/Sprites/Assets/Sounds/Waves/Env_Sprites_Disappear_Chime_07}}
 {id=WavePlayer_11, properties {soundWaveAssetPtr=/Sprites/Assets/Sounds/Waves/Env_Sprites_Disappear_Chime_08}}
 {id=WavePlayer_12, properties {soundWaveAssetPtr=/Sprites/Assets/Sounds/Waves/Env_Sprites_Disappear_Chime_09}}
 {id=WavePlayer_13, properties {soundWaveAssetPtr=/Sprites/Assets/Sounds/Waves/Env_Sprites_Disappear_Chime_10}}
 {id=WavePlayer_14, properties {soundWaveAssetPtr=&Env_Sprites_Disappear_Chime_10}}
 {id=WavePlayer_15, properties {soundWaveAssetPtr=&Env_Sprites_Disappear_Chime_08}}
 {id=WavePlayer_16, properties {soundWaveAssetPtr=/Sprites/Assets/Sounds/Waves/Env_Sprites_Disappear_Chime_05}}
 {id=WavePlayer_17, properties {soundWaveAssetPtr=/Sprites/Assets/Sounds/Waves/Env_Sprites_Disappear_Chime_02}}
 {id=WavePlayer_4, properties {soundWaveAssetPtr=/Sprites/Assets/Sounds/Waves/Env_Sprites_Disappear_Chime_01}}
 {id=WavePlayer_5, properties {soundWaveAssetPtr=&Env_Sprites_Disappear_Chime_02}}
 {id=WavePlayer_6, properties {soundWaveAssetPtr=/Sprites/Assets/Sounds/Waves/Env_Sprites_Disappear_Chime_03}}
 {id=WavePlayer_7, properties {soundWaveAssetPtr=/Sprites/Assets/Sounds/Waves/Env_Sprites_Disappear_Chime_04}}
 {id=WavePlayer_8, properties {soundWaveAssetPtr=&Env_Sprites_Disappear_Chime_05}}
 {id=WavePlayer_9, properties {soundWaveAssetPtr=/Sprites/Assets/Sounds/Waves/Env_Sprites_Disappear_Chime_06}}
]
properties {
 attenuationSettings=&/AS_Env_Sprites_Disappear
 bOverrideConcurrency=true
 maxDistance=600
 soundClassObject=&Fort_Ambient_OneShots
 volumeMultiplier=1
 concurrencyOverrides {maxCount=2, resolutionRule=StopOldest}
}
