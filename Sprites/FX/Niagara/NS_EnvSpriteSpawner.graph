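# Niagara system that scatters a small, capped population of environmental sprites around its owner, with looping spline visuals oriented using an objective position.
# Spawning traces cylinder points to ground, culling steep slopes, certain surfaces and excess particles.
# Sprites bob, turn toward nearby players, and fade with distance from the owner or player proximity, which also triggers a sound.
# A scratch module reads the projectile Data Channel, and a hit timeline feeds material parameters.
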NiagaraSystem /Sprites/FX/Niagara/NS_EnvSpriteSpawner
dependencies {
 usedDataInterfaces [
  &NiagaraDataInterfaceAudioPlayer
  &NiagaraDataInterfaceCollisionQuery
  &NiagaraDataInterfaceColorCurve
  &NiagaraDataInterfaceCurve
  /Script/Niagara.NiagaraDataInterfaceDataChannelRead
  &NiagaraDataInterfaceArrayFloat
 ]
 usedDynamicInputs [&InvertBool]
 usedModules [
  &SystemState
  &self:ReadDeath_01
  &self:SystemSpawnScript.SystemScriptSource.SystemScriptGraph.NiagaraNodeAssignment_0.SetVariables_#1
 ]
}
emitters {Director {
  emitterSpawnScript [&InitializeAudioPlayer(Audio Player=DataInterface() {propertyValues {
      Attenuation=null
      Concurrency=null
      MaxPlaysPerTick=1
      SoundToPlay=null
      bAllowLoopingOneShotSounds=false
      bLimitPlaysPerTick=true
      bOnlyActiveDuringGameplay=true
      bStopWhenComponentIsDestroyed=true
      ConfigurationUserParameter {parameter {name=None, typeDefHandle {registeredTypeIndex=1}, varData []}}
      ParameterNames []
     }})]
  emitterUpdateScript [
   &EmitterState()
   &SpawnRate(SpawnRate=15, Use Spawn Probability=true)
   /SetVariables(Emitter.HideSpline:NiagaraBool=$User.ShowSpline) {isSetParameters=true}
  ]
  particleSpawnScript [
   &InitializeParticle(
    Color=$User.Color
    Lifetime=5
    Mesh Scale Mode=::NewEnumerator1
    Mesh Uniform Scale Max=0.75
    Mesh Uniform Scale Min=0.35
    Sprite Size Mode=::NewEnumerator3
    Uniform Sprite Size=250
    UsePositionOffset=false
    Write Material Random=true
   )
   &ShapeLocation(
    Cylinder Height=750
    Cylinder Radius=&RandomRangeFloat(Minimum=500, Maximum=1750)
    MaskSpawnGroup=false
    Offset Mode=::NewEnumerator0
    Override Seed=false
    Shape Primitive=::NewEnumerator1
    Surface Only=false
    UseOwnerScale=true
   )
   &self:Trace_to_Ground(
    TraceEndWorld:NiagaraPosition=&Subtract_Position(A=$Particles.Position, B=<0 0 500>)
    TraceStartWorld:NiagaraPosition=$Particles.Position
   ) {module=Trace_to_Ground}
   /SetVariables(Particles.Position=&AddVectorToPosition(
     Position=$Output.Trace_to_Ground.CollisionPosWorld
     Vector=&RandomRangeVector(Minimum=<0 0 90>, Maximum=<0 0 100>)
    )) {isSetParameters=true}
   &KillParticles(Kill Particles=&Boolean_NotOperation(Bool=&SetBoolByFloatComparison(
      A=&ClampFloat(Float=&DotProduct(Vector A=$Output.Trace_to_Ground.CollisionNormal, Vector B=<0 0 1>))
      B=0.95
     )))
   &KillParticles(Kill Particles=&SetBoolByFloatComparison(
     A=&Divide_Float(A=&MakeFloatFromInt(Int=$Engine.Emitter.NumParticles), B=8)
     B=1
    )) {module=KillParticles001}
   /SetVariables(Particles.KilledBYSpawnLimit?:NiagaraBool=&SetBoolByFloatComparison(
     A=&Divide_Float(A=&MakeFloatFromInt(Int=$Engine.Emitter.NumParticles), B=8)
     B=1
    )) {isSetParameters=true}
   &KillParticles(Kill Particles=&Boolean_OrOperation(
     Bool1=&SetBoolByIntComparison(A=$Particles.CollisionMaterialIndex, B=0, Comparison Type=::NewEnumerator4)
     Bool2=false
    )) {module=KillParticles002}
   &KillParticles(Kill Particles=&Boolean_OrOperation(
     Bool1=&SetBoolByIntComparison(A=$Particles.CollisionMaterialIndex, B=1, Comparison Type=::NewEnumerator4)
     Bool2=false
    )) {module=KillParticles003}
   &KillParticles(Kill Particles=&Boolean_NotOperation(Bool=$Output.Trace_to_Ground.CollisionValid)) {module=KillParticles004}
   /SetVariables(Particles.Color=&ColorFromCurve(
     ColorCurve=DataInterface() {propertyValues {
       AlphaCurve {keys [<at 0 1 RCIM_Cubic>]}
       BlueCurve {keys [<at 0 0 RCIM_Cubic>,<at 1 0 RCIM_Cubic>]}
       GreenCurve {keys [<at 0 0.0063 RCIM_Cubic>,<at 1 1 RCIM_Cubic>]}
       RedCurve {keys [<at 0 1 RCIM_Cubic>,<at 1 0.0945 RCIM_Cubic>]}
      }}
     CurveIndex=&Divide_Float(A=&MakeFloatFromInt(Int=$Engine.Emitter.NumParticles), B=30)
    )) {enabled=false, isSetParameters=true}
   /SetVariables(Particles.Scale=&VectorFromFloat(Value=&FloatFromCurve(
      FloatCurve=DataInterface() {propertyValues {Curve {keys [<at 0 1>, {arriveTangent=-0.1678, time=1, value=0.9}]}}}
      Scale Curve=&RemapRange(
       Input Value To Remap=&DistanceBetweenPositions(A=$Engine.Owner.Position, B=$Particles.Position)
       Input Min=0
       Input Max=1250
       Output Min=&RandomRangeFloat(Minimum=0.25, Maximum=1)
       Output Max=0.62
      )
     ))) {enabled=false, isSetParameters=true}
   /SetVariables(Particles.Debug:LinearColor=<rgba 0 0 0 1>) {isSetParameters=true}
   &self:WriteDeath_01(Position Array:NiagaraDataInterfaceArrayPosition=$System.DeathArray) {module=WriteDeath_01}
   /SetVariables(Particles.InitialOrientation:Quat4f=&MultiplyQuaternion(
     Quaternion A=<0 0 1 1>
     Quaternion B=&MakeQuatFromAxisAngle(Axis=<0 0 1>, AngleInDegrees=&RandomRangeFloat(Minimum=0, Maximum=360))
    )) {isSetParameters=true}
  ]
  particleUpdateScript [
   &ParticleState(Kill Particles When Lifetime Has Elapsed=false, Let Infinitely Lived Particles Die When Emitter Deactivates=true)
   /SetVariables(Particles.Spline Age:NiagaraFloat=&ModuloFloat(A=$Particles.Age, B=3)) {isSetParameters=true}
   &UpdateMeshOrientation(
    Facing Interpolation=false
    Orientation Method=::NewEnumerator2
    Override Quat=false
    TransformFacingOrigin=false
    TransformFacingPosition=false
    TransformUp=false
   ) {module=UpdateMeshOrientation002}
   /SetVariables(Particles.Reset?:NiagaraBool=&Boolean_NotOperation(Bool=$Particles.NDCHit)) {enabled=false, isSetParameters=true}
   &UpdateMeshOrientation(
    Facing Interpolation=false
    Facing Position:NiagaraPosition=$User.ObjectivePosition
    Orientation Method=::NewEnumerator2
    Override Quat=false
    TransformFacingOrigin=false
    TransformFacingPosition=false
    TransformUp=false
    Write to Orientation:ENiagara_AttributeSamplingApplyOutput=::NewEnumerator1
   ) {module=UpdateMeshOrientation001}
   &Timeline(Play=&Boolean_OrOperation(
     Bool1=&Boolean_OrOperation(
      Bool1=&SetBoolByFloatComparison(
       A=&VectorLength(Vector=&Subtract_Position(A=&MakeVectorFromLinearColorRGB(Color=$NPC.FortniteNPC.PlayerPosition), B=$Particles.Position))
       B=350
       Comparison Type=::NewEnumerator6
      )
      Bool2=&SetBoolByFloatComparison(
       A=&VectorLength(Vector=&Subtract_Position(A=$Engine.Owner.Position, B=$Particles.Position))
       B=1400
      )
     )
     Bool2=&SetBoolByFloatComparison(
      A=&VectorLength(Vector=&Subtract_Position(A=$Engine.Owner.Position, B=$Particles.Position))
      B=1200
     )
    ))
   &ScaleSpriteSize(Uniform Curve Index=&RemapRange(
     Input Value To Remap=&DistanceBetweenPositions(A=$Engine.Owner.Position, B=$Particles.Position)
     Input Min=0
     Input Max=1500
     Output Min=1
     Output Max=0.5
    ))
   &UpdateMeshOrientation(
    Facing Interpolation=false
    Facing Position:NiagaraPosition=&MakeVectorFromLinearColorRGB(Color=$NPC.FortniteNPC.PlayerPosition)
    Orientation Method=::NewEnumerator2
    Override Quat=false
    TransformFacingOrigin=false
    TransformFacingPosition=false
    TransformUp=false
    Write to Orientation:ENiagara_AttributeSamplingApplyOutput=::NewEnumerator1
   )
   &KillParticlesInVolume(Invert Volume=true, Sphere Radius=1500) {module=KillParticlesInVolume001}
   /SetVariables(Particles.Quat:Quat4f=&MultiplyQuaternion(
     Quaternion A=&OrientMeshToVector(
      Look At Direction=&SubtractVectorFromPosition(Position=$User.ObjectivePosition, Vector=Linked() {linkedVariable {name=Particles.Position, type=NiagaraPosition}})
      Reference Axis Vector=<1 0 0>
     )
     Quaternion B=&MakeQuatFromAxisAngle(
      Axis=<1 0 0>
      AngleInDegrees=&Lerp_Float(
       A=0
       B=&RandomRangeFloat(Minimum=30, Maximum=60)
       Alpha=&ModuloFloat(A=$Particles.Age, B=6)
      )
      NormalizeAxis=true
     )
    )) {isSetParameters=true}
   &ScaleColor(
    Scale Mode=::NewEnumerator1
    Scale RGBA=&MakeVector4_FromLinearColor(LinearColor=&ColorFromCurve(ColorCurve=DataInterface() {propertyValues {
        AlphaCurve {keys [<at 0 0 RCIM_Cubic>,<at 0.0951 1 RCIM_Cubic>]}
        BlueCurve {keys [<at 0 1 RCIM_Cubic>,<at 1 1 RCIM_Cubic>]}
        GreenCurve {keys [<at 0 1 RCIM_Cubic>,<at 1 1 RCIM_Cubic>]}
        RedCurve {keys [<at 0 1 RCIM_Cubic>,<at 1 1 RCIM_Cubic>]}
       }}))
    ScaleRGBA=true
   )
   &ScaleColor(
    Curve Index=$Particles.Timeline.NormalizedTime
    Linear Color Curve=DataInterface() {propertyValues {
      AlphaCurve {keys [<at 0 1>,<at 0.4047 0>]}
      BlueCurve {keys [<at 0 1>,<at 1 1>]}
      GreenCurve {keys [<at 0 1>,<at 1 1>]}
      RedCurve {keys [<at 0 1>,<at 1 1>]}
     }}
    Scale Mode=::NewEnumerator2
   ) {module=ScaleColor001}
   &KillParticles(Kill Particles=&Boolean_AndOperation(
     Bool1=&Boolean_AndOperation(
      Bool1=&SetBoolByFloatComparison(
       A=&MakeFloatFromVector4(Vector 4=&MakeVector4_FromLinearColor(LinearColor=$Particles.Color), Channel=::NewEnumerator3)
       B=0.01
       Comparison Type=::NewEnumerator6
      )
      Bool2=&SetBoolByFloatComparison(A=$Particles.Age, B=1)
     )
     Bool2=&SetBoolByFloatComparison(
      A=&VectorLength(Vector=&SubtractVectorFromPosition(
        Position=&MakeVectorFromLinearColorRGB(Color=$NPC.FortniteNPC.PlayerPosition)
        Vector=Linked() {linkedVariable {name=Particles.Initial.Position, type=NiagaraPosition}}
       ))
      B=250
      Comparison Type=::NewEnumerator6
     )
    )) {module=KillParticles005}
   &FN_Sprite_Attributes(
    Camera Offset=75
    Color=&HueShiftLinearColor(
     Hue=0
     Saturation=1.5
     Value=1
     Linear Color=&MultiplyLinearColor(Linear Color=$Particles.Color, Scale Factor=&MakeVector4_FromFloat(Float=1))
    )
    Create Camera Offset Attribute=true
    Create Color Attribute=true
    Create SpriteSize Attribute=true
    SpriteSize=<150 150>
    Write to Module Namespace=true
   )
   /SetVariables(Particles.Position=&AddVectorToPosition(
     Position=$Particles.Position
     Vector=&Lerp_Vector(
      A=<0 0 0>
      B=<0 0 0.3>
      Alpha=&Sine(
       Normalized Angle=&Add_Float(A=$Emitter.Age, B=&RandomRangeFloat(Minimum=0, Maximum=10))
       Period=&RandomRangeFloat(Minimum=5, Maximum=6)
       Scale=1
       Bias=0
      )
     )
    )) {isSetParameters=true}
   &self:ScratchModule_01(
    Distance:NiagaraFloat=&Multiply_Float(A=&MakeFloatFromVector2d(Vector 2D=$Particles.Initial.SpriteSize), B=0.07)
    Reader=DataInterface() {propertyValues {
      Channel=/Game/Effects/Niagara/DataChannels/Projectiles/NDC_FortProjectiles
      bAutoLinkToSpawningNDC=true
      bOnlySpawnOnceOnSubticks=true
      bOverrideSpawnGroupToDataChannelIndex=false
      bReadCurrentFrame=false
      bUpdateSourceDataEveryTick=false
      AccessContext {accessContext {}}
     }}
   ) {module=ScratchModule_01}
   &Timeline(Play=$Particles.NDCHit, Rewind Rate=0.5) {module=Timeline001}
   &ScaleColor(
    Curve Index=$Particles.Timeline001.NormalizedTime
    Linear Color Curve=DataInterface() {propertyValues {
      AlphaCurve {keys [<at 0 1>,<at 0.1276 0>]}
      BlueCurve {keys [<at 0 1>,<at 1 1>]}
      GreenCurve {keys [<at 0 1>,<at 1 1>]}
      RedCurve {keys [<at 0 1>,<at 1 1>]}
     }}
    Scale Mode=::NewEnumerator2
   ) {enabled=false, module=ScaleColor002}
   &KillParticles(Kill Particles=&Boolean_AndOperation(
     Bool1=&SetBoolByIntComparison(A=$Particles.HitCount, B=0, Comparison Type=::NewEnumerator0)
     Bool2=&SetBoolByFloatComparison(A=$Particles.Timeline001.NormalizedTime, B=0.75)
    )) {enabled=false, module=KillParticles006}
   &InterpolateOverTime(
    Rate Of Change:NiagaraFloat=&RandomRangeFloat(Minimum=1.5, Maximum=2)
    Target Vector Value:NiagaraPosition=$Particles.Position
    Variable Type=::NewEnumerator2
   )
   /SetVariables(Particles.Movement:Vector3f=&SubtractVectorFromPosition(
     Position=$Particles.InterpolateOverTime.InterpolateOverTime_V3_.Moving Average
     Vector=Linked() {linkedVariable {name=Particles.Position, type=NiagaraPosition}}
    )) {isSetParameters=true}
   &DynamicMaterialParameters(
    Index 0 Param 1=5
    Index 0 Param 2=150
    Index 1 Param 1=&MakeFloatFromVector(Vector=$Particles.Movement, Channel=::NewEnumerator0)
    Index 1 Param 2=&MakeFloatFromVector(Vector=$Particles.Movement, Channel=::NewEnumerator1)
    Index 1 Param 3=&MakeFloatFromVector(Vector=$Particles.Movement, Channel=::NewEnumerator2)
    Index 2 Param 1=&FloatFromCurve(
     FloatCurve=DataInterface() {propertyValues {Curve {keys [
         {interpMode=RCIM_Cubic, leaveTangent=0.8933, tangentMode=RCTM_Break, time=0, value=0}
         {arriveTangent=-0.0475, tangentMode=RCTM_User, time=3, value=1}
        ]}}}
     CurveIndex=$Particles.Spline Age
    )
    Index 2 Param 2=&FloatFromCurve(
     FloatCurve=DataInterface() {propertyValues {Curve {keys [{interpMode=RCIM_Cubic, leaveTangent=0.6233, tangentMode=RCTM_Break, time=0, value=0}, <at 5 1>]}}}
     CurveIndex=&ModuloFloat(A=$Particles.Age, B=3)
     Scale Curve=60
    )
    Index 3 Param 1=&MakeFloatFromVector(Vector=Linked() {linkedVariable {name=Particles.RayCast, type=NiagaraPosition}}, Channel=::NewEnumerator0)
    Index 3 Param 2=&MakeFloatFromVector(Vector=Linked() {linkedVariable {name=Particles.RayCast, type=NiagaraPosition}}, Channel=::NewEnumerator1)
    Index 3 Param 3=&MakeFloatFromVector(Vector=Linked() {linkedVariable {name=Particles.RayCast, type=NiagaraPosition}}, Channel=::NewEnumerator2)
    Index 3 Param 4=&FloatFromCurve(
     FloatCurve=DataInterface() {propertyValues {Curve {keys [
         {arriveTangent=11.1331, interpMode=RCIM_Cubic, leaveTangent=14.0718, tangentMode=RCTM_Break, time=0, value=0}
         <at 0.5 6 RCIM_Cubic>
         {arriveTangent=-16.2894, interpMode=RCIM_Cubic, leaveTangent=-16.2894, tangentMode=RCTM_User, time=1, value=0}
        ]}}}
     CurveIndex=$Particles.Timeline001.NormalizedTime
    )
    Module.Index0Param0Write=true
    Module.Index0Param1Write=true
    Module.Index0Param2Write=false
    Module.Index0Param3Write=false
    Module.Index1Param0Write=true
    Module.Index1Param1Write=true
    Module.Index1Param2Write=true
    Module.Index1Param3Write=true
    Module.Index2Param0Write=true
    Module.Index2Param1Write=true
    Module.Index2Param2Write=true
    Module.Index2Param3Write=true
    Module.Index3Param0Write=true
    Module.Index3Param1Write=true
    Module.Index3Param2Write=true
    Module.Index3Param3Write=true
    Write Parameter Index 1=true
    Write Parameter Index 2=true
    Write Parameter Index 3=true
   )
   /SetVariables(Particles.Spline Length:Vector3f=&Multiply_VectorByFloat(
     Vector=<1 1 1>
     Float=&RemapRange(
      Input Value To Remap=&VectorLength(Vector=&Subtract_Position(A=$User.ObjectivePosition, B=$Particles.Position))
      Input Min=0
      Input Max=5500
      Output Min=6
      Output Max=6
     )
    )) {isSetParameters=true}
   /SetVariables(Particles.NDCHit:NiagaraBool=$Particles.Timeline001.IsPlaying) {isSetParameters=true}
   /Niagara/Modules/Audio/PlayAudio(
    Audio Player=DataInterface() {propertyValues {
      Attenuation=null
      Concurrency=null
      MaxPlaysPerTick=1
      SoundToPlay=/Sprites/Assets/Sounds/SC_Sprites_Env_OnFadeOut
      bAllowLoopingOneShotSounds=false
      bLimitPlaysPerTick=true
      bOnlyActiveDuringGameplay=true
      bStopWhenComponentIsDestroyed=true
      ConfigurationUserParameter {parameter {name=None, typeDefHandle {registeredTypeIndex=1}, varData []}}
      ParameterNames []
     }}
    Play Audio Trigger:NiagaraBool=&SetBoolByFloatComparison(
     A=&VectorLength(Vector=&SubtractVectorFromPosition(
       Position=&MakeVectorFromLinearColorRGB(Color=$NPC.FortniteNPC.PlayerPosition)
       Vector=Linked() {linkedVariable {name=Particles.Initial.Position, type=NiagaraPosition}}
      ))
     B=250
     Comparison Type=::NewEnumerator6
    )
   )
   /SetVariables(Particles.Spline Color:LinearColor=&MultiplyLinearColor(
     Linear Color=$Particles.Color
     Scale Factor=&MakeVector4(
      X=2
      Y=2
      Z=2
      W=&FloatFromCurve(
       FloatCurve=DataInterface() {propertyValues {Curve {keys [
           {interpMode=RCIM_Cubic, leaveTangent=1.1338, tangentMode=RCTM_Break, time=0, value=0}
           {arriveTangent=0.0274, interpMode=RCIM_Cubic, leaveTangent=-6.767, tangentMode=RCTM_Break, time=0.6, value=1.35}
           {arriveTangent=-0.7503, interpMode=RCIM_Cubic, leaveTangent=-0.5526, tangentMode=RCTM_Break, time=1, value=0.15}
           {interpMode=RCIM_Cubic, tangentMode=RCTM_Break, time=2.5, value=0}
          ]}}}
       CurveIndex=$Particles.Spline Age
       Scale Curve=6
      )
     )
    )) {isSetParameters=true}
   /SetVariables(Particles.Spline Scale:Vector3f=&Multiply_Vector(
     A=$Particles.Spline Length
     B=&MakeVector(X=&Lerp_Float(A=0.15, B=0.25, Alpha=$Particles.Spline Age), Y=0.25, Z=0.25)
    )) {isSetParameters=true}
   /SetVariables(Particles.SpriteOrientation:Quat4f=&LerpQuaternion(
     Quaternion A:Quat4f=$Particles.InitialOrientation
     Quaternion B:Quat4f=$Particles.UpdateMeshOrientation.PreviousFrameQuat
     Lerp Factor:NiagaraFloat=&FloatFromCurve(
      FloatCurve=DataInterface() {propertyValues {Curve {keys [{interpMode=RCIM_Cubic, tangentMode=RCTM_Break, time=0, value=0}, {arriveTangent=-1.1577, tangentMode=RCTM_User, time=1, value=1}]}}}
      CurveIndex=&RemapRange(
       Input Value To Remap=&DistanceBetweenPositions(A=&MakeVectorFromLinearColorRGB(Color=$NPC.FortniteNPC.PlayerPosition), B=$Particles.Position)
       Input Min=750
       Input Max=1000
       Output Min=1
       Output Max=0
      )
     )
     Enable Spherical Linear Interpolation:NiagaraBool=true
    )) {isSetParameters=true}
   /SetVariables(
    Particles.WRS_Color:LinearColor=&MakeLinearColorFromVectorAndFloat(
     "Vector (RGB)"=<1 1 1>
     "Float (Alpha)"=&Multiply_Float(A=$StackContext.Timeline.InverseNormalizedTime, B=0.02)
    )
    Particles.WRS_Dyn:Vector4f=&MakeVector4(
     X=&Add_Float(
      A=&RandomRangeFloat(Minimum=0, Maximum=999)
      B=&FloatFromCurve(
       FloatCurve=DataInterface() {propertyValues {Curve {
          defaultValue=0
          keys [
           {arriveTangent=4.1785, interpMode=RCIM_Cubic, leaveTangent=4.1785, tangentMode=RCTM_User, time=0, value=0}
           {arriveTangent=1.1179, interpMode=RCIM_Cubic, leaveTangent=1.1179, tangentMode=RCTM_User, time=0.2042, value=0.6228}
           {arriveTangent=0.0224, interpMode=RCIM_Cubic, leaveTangent=0.0224, tangentMode=RCTM_User, time=1, value=1}
          ]
         }}}
       Scale Curve=2
      )
     )
     Y=&RandomRangeFloat(Minimum=0, Maximum=1)
     Z=0
     W=0
    )
    Particles.WRS_Size:Vector2f=<30 30>
   ) {isSetParameters=true}
  ]
  rendererClasses [&NiagaraMeshRendererProperties, &NiagaraSpriteRendererProperties]
  renderers [
   {rendererClass=&NiagaraMeshRendererProperties, rendererIndex=0}
   {rendererClass=&NiagaraSpriteRendererProperties, rendererIndex=1}
   {rendererClass=&NiagaraMeshRendererProperties, rendererIndex=2}
   {rendererClass=&NiagaraMeshRendererProperties, rendererIndex=3}
   {rendererClass=&NiagaraSpriteRendererProperties, rendererIndex=4}
  ]
 }}
summary {
 systemName=NS_EnvSpriteSpawner
 emitters [{emitterName=Director, rendererClasses [&NiagaraMeshRendererProperties, &NiagaraSpriteRendererProperties]}]
 userVariables [
  User.Color {type=LinearColor, description="", defaultValue {struct=LinearColor, value=<rgba 0.0938 0.5167 1 1>}}
  User.isWRS {type=NiagaraBool}
  User.ObjectivePosition {type=NiagaraPosition, description="", defaultValue {}}
  User.ShowSpline {type=NiagaraBool, defaultValue=true}
 ]
}
