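# Sound cue that plays the despawn sound for Sprites.
# A Random node picks one of three equally weighted despawn waves, with branch culling exclusion set, and feeds a Modulator for pitch and volume variation.
# It uses the Sprites disappear attenuation asset and the loot-collect sound class.
# A concurrency override caps simultaneous instances, stopping the farthest before preventing new ones.
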SoundCue /Sprites/Assets/Sounds/SC_Sprites_Despawn
root=Modulator_0
nodes [
 {id=Modulator_0, children [Random_2], properties {volumeMin=1}}
 {id=Random_2, children [WavePlayer_12, WavePlayer_13, WavePlayer_14], properties {bShouldExcludeFromBranchCulling=true, weights [1,1,1]}}
 {id=WavePlayer_12, properties {soundWaveAssetPtr=/Sprites/Assets/Sounds/Waves/Sprite_Despawn_01}}
 {id=WavePlayer_13, properties {soundWaveAssetPtr=/Sprites/Assets/Sounds/Waves/Sprite_Despawn_02}}
 {id=WavePlayer_14, properties {soundWaveAssetPtr=/Sprites/Assets/Sounds/Waves/Sprite_Despawn_03}}
]
properties {
 attenuationSettings=&/AS_Env_Sprites_Disappear
 bOverrideConcurrency=true
 maxDistance=600
 soundClassObject=&Fort_Loot_Collect
 volumeMultiplier=1
 concurrencyOverrides {maxCount=5, resolutionRule=StopFarthestThenPreventNew}
}
